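# M_Toy_ResOut is an unlit translucent material for mesh particles, filed with the gameplay pickup effects.
# Particle color RGB feeds the MF_FX_EmissiveAdjustments function to form a glow color.
# A Fresnel with zero base reflect fraction is raised to the ExpFade dynamic parameter and multiplied by particle alpha, producing an edge-weighted fade whose falloff the particle system controls.
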Material /FNGameplayCues/Effects/Fort_Effects/Gameplay/Pickups/Materials/M_Toy_ResOut
graph {
 properties {bUsedWithMeshParticles=true, blendMode=BLEND_Translucent, pixelDepthOffsetMode=PDOM_Legacy, shadingModel=MSM_Unlit, usedShadingModels=Unlit}
 nodes [
  n1 = ParticleColor()
  MaterialFunctionCall(Emissive=@n1.RGB) {materialFunction=&MF_FX_EmissiveAdjustments}
  Multiply(A=@n1.A, B=Power(Base=Fresnel() {baseReflectFraction=0}, Exp=DynamicParameter() {paramNames [ExpFade,Param2,Param3,Param4]}.ExpFade))
 ]
}
